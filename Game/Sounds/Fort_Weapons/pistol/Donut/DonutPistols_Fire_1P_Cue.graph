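# Sound cue for the first-person firing sound of the Donut pistols, layering a close gunshot with a delayed shell casing drop.
# A mixer combines one of three random close-fire waves, passed through a default modulator, with a much quieter casing branch.
# That branch briefly delays one of five casing drops borrowed from the burst-fire SMG pistol, randomly lowered in pitch and volume.
# It uses first-person weapon attenuation, sound class and gunfire concurrency.
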SoundCue /Game/Sounds/Fort_Weapons/pistol/Donut/DonutPistols_Fire_1P_Cue
root=Mixer_0
nodes [
 {id=Delay_0, children [Modulator_3], properties {delayMax=0.8, delayMin=0.7}}
 {id=Mixer_0, children [Modulator_1,Delay_0], properties {inputVolume [1,0.2]}}
 {id=Modulator_1, children [Random_3]}
 {id=Modulator_3, children [Random_5], properties {pitchMax=0.9, pitchMin=0.8, volumeMax=0.9, volumeMin=0.7}}
 {id=Random_3, children [WavePlayer_21, WavePlayer_22, WavePlayer_23], properties {weights [1,1,1]}}
 {id=Random_5, children [WavePlayer_16, WavePlayer_17, WavePlayer_18, WavePlayer_19, WavePlayer_20], properties {weights [1,1,1,1,1]}}
 {id=WavePlayer_16, properties {soundWaveAssetPtr=/Game/Athena/Sounds/Weapons/Pistol_BurstFireSMG/Pistol_BurstFireSMG_Bullet_Casings_Drop_01}}
 {id=WavePlayer_17, properties {soundWaveAssetPtr=/Game/Athena/Sounds/Weapons/Pistol_BurstFireSMG/Pistol_BurstFireSMG_Bullet_Casings_Drop_02}}
 {id=WavePlayer_18, properties {soundWaveAssetPtr=/Game/Athena/Sounds/Weapons/Pistol_BurstFireSMG/Pistol_BurstFireSMG_Bullet_Casings_Drop_03}}
 {id=WavePlayer_19, properties {soundWaveAssetPtr=/Game/Athena/Sounds/Weapons/Pistol_BurstFireSMG/Pistol_BurstFireSMG_Bullet_Casings_Drop_04}}
 {id=WavePlayer_20, properties {soundWaveAssetPtr=/Game/Athena/Sounds/Weapons/Pistol_BurstFireSMG/Pistol_BurstFireSMG_Bullet_Casings_Drop_05}}
 {id=WavePlayer_21, properties {soundWaveAssetPtr=&/DonutPistols_FireClose_06}}
 {id=WavePlayer_22, properties {soundWaveAssetPtr=&/DonutPistols_FireClose_07}}
 {id=WavePlayer_23, properties {soundWaveAssetPtr=&/DonutPistols_FireClose_08}}
]
properties {
 attenuationSettings=&Fort_Weapon_Projectile_FirstPerson_Athena_Att
 maxDistance=2097152
 soundClassObject=&Fort_WeaponRanged_Fire_1P_PSM
 volumeMultiplier=1.05
 attenuationOverrides {distanceAlgorithm=NaturalSound, falloffDistance=5000, attenuationShapeExtents {x=1000}}
 concurrencySet [&Weapon_Gunfire_FirstPerson_Conc]
}
