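# Sound cue for Donut pistol gunfire, built from close, medium and distant layers.
# A mixer blends the three branches at descending input volumes, each with its own attenuation asset and a random pick among wave variations.
# Close and medium layers use the third-person ranged fire sound class, while the distant layer uses the distant fire class; modulators boost medium and distant volume.
# Local and global gunfire concurrency sets govern simultaneous playback.
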SoundCue /Game/Sounds/Fort_Weapons/pistol/Donut/DonutPistols_Fire_Cue
root=Mixer_17
nodes [
 {
  id=Attenuation_0
  children [Modulator_3]
  properties {attenuationSettings=&Fort_Weapon_Projectile_Medium_Mono_Athena_Att}
 }
 {
  id=Attenuation_34
  children [Modulator_4]
  properties {attenuationSettings=&Fort_Weapon_Projectile_Distant_Mono_Athena_Att}
 }
 {id=Attenuation_49, children [Modulator_2], properties {attenuationSettings=&Fort_Weapon_Projectile_Close_Athena_Att}}
 {id=Mixer_17, children [SoundClass_32, SoundClass_0, SoundClass_35], properties {inputVolume [1,0.7,0.4]}}
 {id=Modulator_2, children [Random_6]}
 {id=Modulator_3, children [Random_4], properties {volumeMax=1.9, volumeMin=1.9}}
 {id=Modulator_4, children [Random_5], properties {volumeMax=1.3, volumeMin=1.3}}
 {id=Random_4, children [WavePlayer_14, WavePlayer_15, WavePlayer_16, WavePlayer_17], properties {weights [1,1,1,1]}}
 {id=Random_5, children [WavePlayer_18, WavePlayer_19, WavePlayer_20], properties {weights [1,1,1]}}
 {id=Random_6, children [WavePlayer_22, WavePlayer_23, WavePlayer_24], properties {weights [1,1,1]}}
 {
  id=SoundClass_0
  children [Attenuation_0]
  properties {soundClassOverride=&Fort_WeaponRanged_Fire_3P_PSM}
 }
 {
  id=SoundClass_32
  children [Attenuation_49]
  properties {soundClassOverride=&Fort_WeaponRanged_Fire_3P_PSM}
 }
 {
  id=SoundClass_35
  children [Attenuation_34]
  properties {soundClassOverride=&Fort_WeaponRanged_Fire_Distant}
 }
 {id=WavePlayer_14, properties {soundWaveAssetPtr=&/DonutPistols_FireMedium_01}}
 {id=WavePlayer_15, properties {soundWaveAssetPtr=&/DonutPistols_FireMedium_02}}
 {id=WavePlayer_16, properties {soundWaveAssetPtr=&/DonutPistols_FireMedium_03}}
 {id=WavePlayer_17, properties {soundWaveAssetPtr=&/DonutPistols_FireMedium_04}}
 {id=WavePlayer_18, properties {soundWaveAssetPtr=&/DonutPistols_FireDistant_01}}
 {id=WavePlayer_19, properties {soundWaveAssetPtr=&/DonutPistols_FireDistant_02}}
 {id=WavePlayer_20, properties {soundWaveAssetPtr=&/DonutPistols_FireDistant_03}}
 {id=WavePlayer_22, properties {soundWaveAssetPtr=&/DonutPistols_FireClose_06}}
 {id=WavePlayer_23, properties {soundWaveAssetPtr=&/DonutPistols_FireClose_07}}
 {id=WavePlayer_24, properties {soundWaveAssetPtr=&/DonutPistols_FireClose_08}}
]
properties {
 maxDistance=27500
 volumeMultiplier=1.05
 attenuationOverrides {distanceAlgorithm=NaturalSound, falloffDistance=5000, attenuationShapeExtents {x=1000}}
 concurrencySet [&Weapon_Gunfire_Local_Conc, &Weapon_Gunfire_Global_Conc]
}
