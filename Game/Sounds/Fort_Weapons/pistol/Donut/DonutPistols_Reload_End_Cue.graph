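# Sound cue for the end of the Donut pistols' reload, played as weapon foley.
# A Random node picks one of two equally weighted ReloadEnd wave variants and feeds it through a root Modulator left at default settings.
# The cue uses the shared weapon reload attenuation asset and the weapon foley sound class, with lowered volume and a concurrency override that prevents new instances once the limit is reached.
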SoundCue /Game/Sounds/Fort_Weapons/pistol/Donut/DonutPistols_Reload_End_Cue
root=Modulator_1
nodes [
 {id=Modulator_1, children [Random_2]}
 {id=Random_2, children [WavePlayer_6,WavePlayer_7], properties {weights [1,1]}}
 {id=WavePlayer_6, properties {soundWaveAssetPtr=&/DonutPistols_ReloadEnd_03}}
 {id=WavePlayer_7, properties {soundWaveAssetPtr=&/DonutPistols_ReloadEnd_04}}
]
properties {
 attenuationSettings=&Fort_Weapon_Reload_Att
 bOverrideConcurrency=true
 maxDistance=2250
 soundClassObject=&Fort_WeaponFoley
 volumeMultiplier=0.752
 attenuationOverrides {falloffDistance=4100, attenuationShapeExtents {x=900}}
 concurrencyOverrides {resolutionRule=PreventNew}
}
