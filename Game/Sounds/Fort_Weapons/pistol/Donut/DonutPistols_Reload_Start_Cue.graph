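# Sound cue for the start of the reload on the Donut pistols.
# It picks one of three reload-start wave variants at random with equal weight and passes it through a modulator.
# It plays as weapon foley with the shared weapon reload attenuation and a reduced volume.
# A concurrency override blocks new instances rather than stopping playing ones.
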SoundCue /Game/Sounds/Fort_Weapons/pistol/Donut/DonutPistols_Reload_Start_Cue
root=Modulator_2
nodes [
 {id=Modulator_2, children [Random_1]}
 {id=Random_1, children [WavePlayer_3,WavePlayer_4,WavePlayer_5], properties {weights [1,1,1]}}
 {id=WavePlayer_3, properties {soundWaveAssetPtr=&/DonutPistols_ReloadStart_01}}
 {id=WavePlayer_4, properties {soundWaveAssetPtr=&/DonutPistols_ReloadStart_02}}
 {id=WavePlayer_5, properties {soundWaveAssetPtr=&/DonutPistols_ReloadStart_03}}
]
properties {
 attenuationSettings=&Fort_Weapon_Reload_Att
 bOverrideConcurrency=true
 maxDistance=2250
 soundClassObject=&Fort_WeaponFoley
 volumeMultiplier=0.702
 attenuationOverrides {falloffDistance=4100, attenuationShapeExtents {x=900}}
 concurrencyOverrides {resolutionRule=PreventNew}
}
